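# Sound cue for the Feral variant of the critical death hit marker, a notification that by its naming plays for the instigating player.
# A mixer layers three briefly delayed, pitched-up branches: a random pick of two elimination sounds, a shared death stinger and a louder Feral elimination stinger.
# It uses the instigator hit-notification sound class and attenuation, with a critical hit-notification stop-oldest concurrency rule.
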SoundCue /Game/Athena/Sounds/HitNotifications/Feral/HitMarker_Death_Critical_Feral_Cue
root=Mixer_0
nodes [
 {id=Delay_0, children [Modulator_2], properties {delayMax=0.15, delayMin=0.15}}
 {id=Delay_1, children [Modulator_0], properties {delayMax=0.15, delayMin=0.15}}
 {id=Delay_2, children [Modulator_1], properties {delayMax=0.1, delayMin=0.1}}
 {id=Mixer_0, children [Delay_2,Delay_1,Delay_0], properties {inputVolume [1,1.2,1.5]}}
 {id=Modulator_0, children [WavePlayer_1], properties {pitchMax=1.4, pitchMin=1.4, volumeMax=1.3, volumeMin=1.3}}
 {id=Modulator_1, children [Random_0], properties {pitchMax=1.1, pitchMin=1.1}}
 {id=Modulator_2, children [WavePlayer_4], properties {pitchMax=1.1, pitchMin=1.1, volumeMax=1.5, volumeMin=1.5}}
 {id=Random_0, children [WavePlayer_2,WavePlayer_3], properties {bShouldExcludeFromBranchCulling=true, weights [1,1]}}
 {id=WavePlayer_1, properties {soundWaveAssetPtr=/Game/Athena/Sounds/HitNotifications/HitMarker_Death_Stinger_01}}
 {id=WavePlayer_2, properties {soundWaveAssetPtr=/Game/Athena/Sounds/HitNotifications/HitMarker_Elim_01}}
 {id=WavePlayer_3, properties {soundWaveAssetPtr=/Game/Athena/Sounds/HitNotifications/HitMarker_Elim_02}}
 {id=WavePlayer_4, properties {soundWaveAssetPtr=&/FN_Feral_Stinger_Stem_Elimination_Short}}
]
properties {
 attenuationSettings=/Game/Athena/Sounds/HitNotifications/Fort_HitMarker_FP_Instigate_Att
 maxDistance=2097152
 soundClassObject=&Fort_2DHitNotifications_Instigate_PSM
 volumeMultiplier=0.9
 concurrencyOverrides {maxCount=2, resolutionRule=StopOldest}
 concurrencySet [/Game/Sounds/Concurrency_Rules/HitNotifications_Critical_StopOldest]
}
