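# Sound cue for the Feral variant of the hit notification heard on receiving a death hit.
# A quality switch picks the content: the top two levels mix a random rifle bullet impact with a briefly delayed big-hit marker, the second drawing on fewer impact variations.
# The lower two levels share a single pre-optimized wave through a modulator.
# Receive hit-notification sound class, attenuation and stop-oldest concurrency settings govern playback.
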SoundCue /Game/Athena/Sounds/HitNotifications/Feral/HitMarker_Death_Receive_Feral_Cue
root=QualityLevel_0
nodes [
 {id=Delay_0, children [WavePlayer_10], properties {delayMax=0.2, delayMin=0.2}}
 {id=Delay_1, children [WavePlayer_5], properties {delayMax=0.2, delayMin=0.2}}
 {id=Mixer_0, children [Delay_1,Random_0], properties {inputVolume [1,1]}}
 {id=Mixer_1, children [Delay_0,Random_1], properties {inputVolume [1,1]}}
 {id=Modulator_0, children [WavePlayer_2]}
 {id=QualityLevel_0, children [Mixer_1,Mixer_0,Modulator_0,Modulator_0]}
 {id=Random_0, children [WavePlayer_0,WavePlayer_1], properties {bShouldExcludeFromBranchCulling=true, weights [1,1]}}
 {
  id=Random_1
  children [WavePlayer_6,WavePlayer_7,WavePlayer_8,WavePlayer_9]
  properties {bShouldExcludeFromBranchCulling=true, weights [1,1,1,1]}
 }
 {id=WavePlayer_0, properties {soundWaveAssetPtr=/Game/Sounds/Fort_Weapons/Impacts/player_bulletimpact_rifle_01}}
 {id=WavePlayer_1, properties {soundWaveAssetPtr=/Game/Sounds/Fort_Weapons/Impacts/player_bulletimpact_rifle_02}}
 {id=WavePlayer_10, properties {soundWaveAssetPtr=/Game/Athena/Sounds/HitNotifications/sfx_hitmarker_bighit}}
 {id=WavePlayer_2, properties {soundWaveAssetPtr=/Game/Athena/Sounds/HitNotifications/HitMarker_Death_Receive_opt}}
 {id=WavePlayer_5, properties {soundWaveAssetPtr=&sfx_hitmarker_bighit}}
 {id=WavePlayer_6, properties {soundWaveAssetPtr=&player_bulletimpact_rifle_01}}
 {id=WavePlayer_7, properties {soundWaveAssetPtr=&player_bulletimpact_rifle_02}}
 {id=WavePlayer_8, properties {soundWaveAssetPtr=/Game/Sounds/Fort_Weapons/Impacts/player_bulletimpact_rifle_03}}
 {id=WavePlayer_9, properties {soundWaveAssetPtr=/Game/Sounds/Fort_Weapons/Impacts/player_bulletimpact_rifle_04}}
]
properties {
 attenuationSettings=/Game/Athena/Sounds/HitNotifications/Fort_HitMarker_FP_Instigate_Att
 maxDistance=2097152
 soundClassObject=/Game/Sounds/FortSoundClassesAndModes/Fort_2DHitNotifications_Receive_PSM
 volumeMultiplier=0.494
 concurrencyOverrides {maxCount=2, resolutionRule=StopOldest}
 concurrencySet [&HitNotifications_TwoMaximum_StopOldest_VolumeScale]
}
