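# Sound cue for the Feral hit marker, a hit notification that layers two randomized sounds.
# A mixer plays one of three body hit waves, pitched up by a fixed amount, alongside one of three Feral crit impact indicators at half volume.
# A very short delay staggers the crit layer behind the body hit.
# First-person instigator attenuation, a 2D notification sound class and a two-voice stop-oldest concurrency limit govern playback.
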SoundCue /Game/Athena/Sounds/HitNotifications/Feral/HitMarker_Feral_Cue
root=Mixer_0
nodes [
 {id=Delay_0, children [Modulator_2], properties {delayMax=0.03, delayMin=0.03}}
 {id=Mixer_0, children [Modulator_0,Delay_0], properties {inputVolume [1,0.5]}}
 {id=Modulator_0, children [Random_1], properties {pitchMax=1.3, pitchMin=1.3, volumeMax=1, volumeMin=1}}
 {id=Modulator_2, children [Random_2], properties {pitchMax=1, pitchMin=1, volumeMax=1, volumeMin=1}}
 {id=Random_1, children [WavePlayer_5,WavePlayer_6,WavePlayer_7], properties {weights [1,1,1]}}
 {id=Random_2, children [WavePlayer_8, WavePlayer_9, WavePlayer_10], properties {weights [1,1,1]}}
 {id=WavePlayer_10, properties {soundWaveAssetPtr=&/FERAL_CritImpact_Indicator_06}}
 {id=WavePlayer_5, properties {soundWaveAssetPtr=&/HitMarker_Body_01_edit}}
 {id=WavePlayer_6, properties {soundWaveAssetPtr=&/HitMarker_Body_02_edit}}
 {id=WavePlayer_7, properties {soundWaveAssetPtr=&/HitMarker_Body_03_edit}}
 {id=WavePlayer_8, properties {soundWaveAssetPtr=&/FERAL_CritImpact_Indicator_04}}
 {id=WavePlayer_9, properties {soundWaveAssetPtr=&/FERAL_CritImpact_Indicator_05}}
]
properties {
 attenuationSettings=/Game/Athena/Sounds/HitNotifications/Fort_HitMarker_FP_Instigate_Att
 maxDistance=2097152
 soundClassObject=&Fort_2DHitNotifications_Instigate_PSM
 volumeMultiplier=1
 concurrencyOverrides {maxCount=2, resolutionRule=StopOldest}
 concurrencySet [&HitNotifications_TwoMaximum_StopOldest_VolumeScale]
}
