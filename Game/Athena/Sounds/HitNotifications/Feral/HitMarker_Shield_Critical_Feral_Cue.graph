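# Sound cue for the Feral variant of the critical shield hit-marker notification, using instigator hit-notification sound class and attenuation.
# A mixer layers one of four randomly chosen, equally weighted shield hit waves, pitched slightly up, with a louder accent.
# That accent reuses a cash pickup tone, pitched far up and briefly delayed behind the hit.
# Its concurrency set is the shared two-maximum, stop-oldest hit notification one.
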SoundCue /Game/Athena/Sounds/HitNotifications/Feral/HitMarker_Shield_Critical_Feral_Cue
root=Mixer_0
nodes [
 {id=Delay_1, children [Modulator_1], properties {delayMax=0.04, delayMin=0.04}}
 {id=Mixer_0, children [Modulator_0,Delay_1], properties {inputVolume [1,1.5]}}
 {id=Modulator_0, children [Random_0], properties {pitchMax=1.2, pitchMin=1.3, volumeMax=1, volumeMin=1}}
 {id=Modulator_1, children [WavePlayer_4], properties {pitchMax=3, pitchMin=3, volumeMax=1, volumeMin=1}}
 {
  id=Random_0
  children [WavePlayer_0,WavePlayer_1,WavePlayer_2,WavePlayer_3]
  properties {bShouldExcludeFromBranchCulling=true, weights [1,1,1,1]}
 }
 {id=WavePlayer_0, properties {soundWaveAssetPtr=/Game/Athena/Sounds/HitNotifications/HitMarker_Shield_3P_01}}
 {id=WavePlayer_1, properties {soundWaveAssetPtr=/Game/Athena/Sounds/HitNotifications/HitMarker_Shield_3P_02}}
 {id=WavePlayer_2, properties {soundWaveAssetPtr=/Game/Athena/Sounds/HitNotifications/HitMarker_Shield_3P_03}}
 {id=WavePlayer_3, properties {soundWaveAssetPtr=/Game/Athena/Sounds/HitNotifications/HitMarker_Shield_3P_04}}
 {id=WavePlayer_4, properties {soundWaveAssetPtr=/Game/Athena/Sounds/Items/Wads/Cash_PickUp_Tone_Med}}
]
properties {
 attenuationSettings=/Game/Athena/Sounds/HitNotifications/Fort_HitMarker_FP_Instigate_Att
 maxDistance=2097152
 soundClassObject=&Fort_2DHitNotifications_Instigate_PSM
 volumeMultiplier=0.8
 concurrencyOverrides {maxCount=2, resolutionRule=StopOldest}
 concurrencySet [&HitNotifications_TwoMaximum_StopOldest_VolumeScale]
}
