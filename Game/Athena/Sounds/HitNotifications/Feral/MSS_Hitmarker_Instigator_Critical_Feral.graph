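# Stereo one-shot MetaSound for the Feral critical hitmarker heard by the hit's instigator.
# It runs three MSS_Hitmarker_Instigator_Parent instances fed by source azimuth: body sounds play immediately, while Feral crit impacts and a pitch-shifted cash-pickup tone follow after a short shared delay.
# The delayed layers are submixed, then boosted and mixed with the body.
# A trigger accumulator signals OnFinished once all three layers end, and concurrency stops the oldest overlapping hit.
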MetaSoundSource /Game/Athena/Sounds/HitNotifications/Feral/MSS_Hitmarker_Instigator_Critical_Feral
inputs [
 Body Sounds {type=WaveAsset:Array}
 Crit Layer Sounds {type=WaveAsset:Array}
 EnemyImpactSounds {type=WaveAsset:Array}
 UE.Source.OnPlay {type=Trigger}
 UE.Source.Orientation.Azimuth {type=Float}
 UE.Source.Orientation.Elevation {type=Float}
]
interfaces [UE.OutputFormat.Stereo:1.0, UE.Source.OneShot:1.0, UE.Source.Orientation:1.0, UE.Source:1.1]
outputs [
 "UE.OutputFormat.Stereo.Audio:0" {type=Audio}
 "UE.OutputFormat.Stereo.Audio:1" {type=Audio}
 UE.Source.OneShot.OnFinished {type=Trigger}
]
pages [Default {
  input_defaults {
   UE.Source.OnPlay=null
   UE.Source.Orientation.Azimuth=0
   UE.Source.Orientation.Elevation=0.14
   Body Sounds [
    &/HitMarker_Body_01_edit
    &/HitMarker_Body_02_edit
    &/HitMarker_Body_03_edit
   ]
   Crit Layer Sounds [/Game/Athena/Sounds/Items/Wads/Cash_PickUp_Tone_Med]
   EnemyImpactSounds [
    &/FERAL_CritImpact_Indicator_06
    &/FERAL_CritImpact_Indicator_05
    &/FERAL_CritImpact_Indicator_04
   ]
  }
  nodes [
   TriggerDelay_0 = UE.Trigger Delay(In=^UE.Source.OnPlay) {also {defaults {Delay Time=0.1}}}
   n1 = /Game/Athena/Sounds/HitNotifications/MSS_Hitmarker_Instigator_Parent(Sounds=^EnemyImpactSounds, UE.Source.Orientation.Azimuth=^UE.Source.Orientation.Azimuth, UE.Source.OnPlay=@TriggerDelay_0.Out)
   n2 = &MSS_Hitmarker_Instigator_Parent(
    Sounds=^Crit Layer Sounds
    UE.Source.Orientation.Azimuth=UE.Reroute(Value=^UE.Source.Orientation.Azimuth).Value
    UE.Source.OnPlay=@TriggerDelay_0.Out
   ) {also {defaults {Max Pitch=12, Min Pitch=12}}}
   AudioMixerAudioMixerStereo2_1 = "AudioMixer.Audio Mixer (Stereo, 2)"(
    In 0 L=@n1."UE.OutputFormat.Stereo.Audio:0"
    In 0 R=@n1."UE.OutputFormat.Stereo.Audio:1"
    In 1 L=@n2."UE.OutputFormat.Stereo.Audio:0"
    In 1 R=@n2."UE.OutputFormat.Stereo.Audio:1"
   ) {also {defaults {Gain 0=0.7, Gain 1=0.8}}}
   n3 = &MSS_Hitmarker_Instigator_Parent(
    Sounds=^Body Sounds
    UE.Source.OnPlay=^UE.Source.OnPlay
    UE.Source.Orientation.Azimuth=UE.Reroute(Value=^UE.Source.Orientation.Azimuth).Value
   ) {also {defaults {Max Pitch=1.3, Min Pitch=1.3}}}
   AudioMixerAudioMixerStereo2_0 = "AudioMixer.Audio Mixer (Stereo, 2)"(
    In 1 L=@AudioMixerAudioMixerStereo2_1.Out L
    In 1 R=@AudioMixerAudioMixerStereo2_1.Out R
    In 0 L=@n3."UE.OutputFormat.Stereo.Audio:0"
    In 0 R=@n3."UE.OutputFormat.Stereo.Audio:1"
   ) {also {defaults {Gain 1=1.7}}}
   TriggerAccumulatorTriggerAccumulate3_0 = "TriggerAccumulator.Trigger Accumulate (3)"(In 1=@n1.UE.Source.OneShot.OnFinished, In 0=@n3.UE.Source.OneShot.OnFinished, In 2=@n2.UE.Source.OneShot.OnFinished)
  ]
  links [
   {dst="@UE.OutputFormat.Stereo.Audio:0", src=AudioMixerAudioMixerStereo2_0.Out L}
   {dst="@UE.OutputFormat.Stereo.Audio:1", src=AudioMixerAudioMixerStereo2_0.Out R}
   {dst="@UE.Source.OneShot.OnFinished", src=TriggerAccumulatorTriggerAccumulate3_0.Out}
  ]
 }]
properties {
 assetImportData=&self:AssetImportData
 attenuationSettings=/Game/Athena/Sounds/HitNotifications/Fort_HitMarker_FP_Instigate_Att
 numChannels=2
 outputFormat=Stereo
 sampleRate=48000
 soundClassObject=&Fort_2DHitNotifications_Instigate_PSM
 volume=0.6
 concurrencySet [&HitNotifications_TwoMaximum_StopOldest_VolumeScale]
}
